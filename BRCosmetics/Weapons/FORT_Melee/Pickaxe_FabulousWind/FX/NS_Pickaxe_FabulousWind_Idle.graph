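# Idle Niagara effect for the FabulousWind pickaxe that sheds small bursts of pink-to-purple feathers.
# A single Feather emitter spawns particles in an offset cylinder, launching and orienting them along a blended outward vector plus inherited velocity.
# Aerodynamic drag with lift and gravity makes them flutter, drawn by mesh renderers and a sprite aligned to mesh orientation.
# User.bisSwinging zeroes the mesh scale curve, collapsing feather meshes during swings.
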NiagaraSystem /BRCosmetics/Weapons/FORT_Melee/Pickaxe_FabulousWind/FX/NS_Pickaxe_FabulousWind_Idle
dependencies {usedDataInterfaces [&NiagaraDataInterfaceCurve], usedDynamicInputs [], usedModules [&SystemState]}
emitters {Feather {
  emitterSpawnScript [/SetVariables(Emitter.PositionOffset:Vector3f=<0 -50 10>) {isSetParameters=true}]
  emitterUpdateScript [
   &EmitterState()
   &SpawnBurst_Instantaneous(
    Spawn Count=&RandomRangeInt(Minimum=0, Maximum=2)
    Use Spawn Probability=false
    UseLoopCountLimit=false
   )
  ]
  particleSpawnScript [
   &InitializeParticle(
    Color Maximum=<rgba 2 0 0.1667 1>
    Color Minimum=<rgba 2.1144 0 3 1>
    Color Mode=::NewEnumerator2
    Lifetime Min=0.7
    Lifetime Mode=::NewEnumerator1
    Mesh Scale Mode=::NewEnumerator1
    Mesh Uniform Scale Max=1.25
    Mesh Uniform Scale Min=0.7
    Position Offset=$Emitter.PositionOffset
    Position Offset Coordinate Space=::Local
    Sprite Size Mode=::NewEnumerator3
    Uniform Sprite Size=2
    UsePositionOffset=true
    Write Material Random=true
   )
   &ShapeLocation(
    Cylinder Height=20
    Cylinder Radius=5
    MaskSpawnGroup=false
    Offset Mode=::NewEnumerator0
    Override Seed=false
    Rotation Axis=&NormalizeVector(Vector To Normalize=$Emitter.PositionOffset, Safe Normalize=true, Safe Fallback Threshold=0, Safe Fallback Vector=<0 0 0.001>)
    Rotation Mode=::NewEnumerator1
    Shape Primitive=::NewEnumerator1
    Surface Only=false
    UseOwnerScale=true
   )
   /SetVariables(Particles.FeatherVector:Vector3f=&Lerp_Vector(
     A=&Lerp_Vector(
      A=$Particles.ShapeLocation.ShapeNormal
      B=&TransformVector(Vector=<0.5 0 1>, Destination Space=::Simulation)
      Alpha=0.5
     )
     B=&TransformVector(
      Vector=&NormalizeVector(Vector To Normalize=$Emitter.PositionOffset, Safe Normalize=true, Safe Fallback Threshold=0, Safe Fallback Vector=<0 0 0.001>)
      Destination Space=::Simulation
     )
     Alpha=0.7
    )) {isSetParameters=true}
   &InitialMeshOrientation(Mesh Axis To Orient=<0 1 0>, Mesh Orientation Mode=::NewEnumerator2, Orientation Vector=$Particles.FeatherVector, Use Rotation Vector=false) {module=InitialMeshOrientation001}
   &InitialMeshOrientation(
    Mesh Axis To Orient=<0 1 0>
    Mesh Orientation Mode=::NewEnumerator2
    Orientation Coordinate Space=::Local
    Orientation Vector=<0 1 0>
    Rotation=&RandomRangeVector(Minimum=<0 -1 0>, Maximum=<0 1 0>)
    Use Rotation Vector=true
   ) {enabled=false}
   &AddVelocity(Rotation Coordinate Space=::World, Velocity=$Particles.FeatherVector, Velocity Speed Scale=50)
   &InheritVelocity(
    Inherited Velocity Amount Scale=&VectorFromFloat(Value=1)
    Inherited Velocity Speed Limit=300
    Speed Limit=true
    UseSpeedThreshold=true
   )
   &CameraOffset(Camera Offset Amount=3)
  ]
  particleUpdateScript [
   &AerodynamicDrag(
    Aerodynamic Drag:NiagaraFloat=&RandomRangeFloat(Minimum=0.2, Maximum=0.6)
    Aerodynamic Rotational Drag=0.05
    Drag Shape Curve=DataInterface() {propertyValues {Curve {keys [
        {interpMode=RCIM_Cubic, tangentMode=RCTM_User, time=0, value=0.1}
        {arriveTangent=2.0799, interpMode=RCIM_Cubic, leaveTangent=2.0799, tangentMode=RCTM_User, time=1, value=1}
       ]}}}
    Lift Contribution:NiagaraFloat=&RandomRangeFloat(Minimum=0.3, Maximum=1)
    Lift Shape Curve=DataInterface() {propertyValues {Curve {keys [
        {interpMode=RCIM_Cubic, tangentMode=RCTM_User, time=0, value=0}
        {interpMode=RCIM_Cubic, tangentMode=RCTM_User, time=0.5, value=1}
        {tangentMode=RCTM_User, time=1, value=0}
       ]}}}
    Pivot Offset Maximum:Vector3f=<4 0.75 0.25>
    Pivot Offset Minimum:Vector3f=<-4 -0.75 -0.25>
    Rotation Strength:NiagaraFloat=&RandomRangeFloat(Minimum=0.75, Maximum=2)
    Use Linear Drag=true
    Use Rotational Drag=true
   )
   &GravityForce(Gravity=<0 0 -122.5>)
   &SolveForcesAndVelocity(Clamp Velocity=false, Limit Acceleration=false)
   &ParticleState()
   &DynamicMaterialParameters(
    Index 0 Param 1=&RandomRangeFloat(Minimum=0, Maximum=0.1)
    Index 0 Param 2=&FloatFromCurve(FloatCurve=DataInterface() {propertyValues {Curve {keys [<at 0.75 1>,<at 1 0>]}}})
    Index 0 Param 4=&RandomRangeFloat(Minimum=0, Maximum=0.5)
    Module.Param0WriteEnabled=true
    Module.Param1WriteEnabled=true
    Module.Param2WriteEnabled=true
    Module.Param3WriteEnabled=true
   )
   &AlignSpriteToMeshOrientation(Mesh Orientation Relative Sprite Facing Vector:Vector3f=<0 1 0>, SetAlignment=true, SetFacing=true)
   &ScaleMeshSize(Scale Factor=&Multiply_VectorByFloat(
     Vector=<1 1 1>
     Float=&FloatFromCurve(
      FloatCurve=DataInterface() {propertyValues {Curve {keys [
          {arriveTangent=4.1785, interpMode=RCIM_Cubic, leaveTangent=4.1785, tangentMode=RCTM_User, time=0, value=0.55}
          {arriveTangent=0.9, interpMode=RCIM_Cubic, leaveTangent=0.9, time=0.15, value=1}
          {arriveTangent=0.1016, interpMode=RCIM_Cubic, leaveTangent=0.1016, tangentMode=RCTM_User, time=0.5, value=1}
          {interpMode=RCIM_Cubic, leaveTangent=-1.5153, time=1, value=0}
         ]}}}
      Scale Curve=&MakeCustomFloatFromBool(Bool=$User.bisSwinging, True Float=0, False Float=1.25)
     )
    ))
  ]
  rendererClasses [&NiagaraMeshRendererProperties, &NiagaraSpriteRendererProperties]
  renderers [
   {rendererClass=&NiagaraMeshRendererProperties, rendererIndex=0}
   {rendererClass=&NiagaraSpriteRendererProperties, rendererIndex=1}
   {rendererClass=&NiagaraMeshRendererProperties, rendererIndex=2}
  ]
 }}
summary {
 systemName=NS_Pickaxe_FabulousWind_Idle
 emitters [{emitterName=Feather, rendererClasses [&NiagaraMeshRendererProperties, &NiagaraSpriteRendererProperties]}]
 userVariables [
  User.bIsFrontEnd {type=NiagaraBool}
  User.bIsFrontEndPreview {type=NiagaraBool}
  User.bisSwinging {type=NiagaraBool}
  User.IsSwing {type=NiagaraBool}
 ]
}
