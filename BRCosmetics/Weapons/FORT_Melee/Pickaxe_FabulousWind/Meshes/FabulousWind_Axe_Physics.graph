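# Physics asset for the FabulousWind pickaxe mesh, setting up capsule bodies and joint limits for its dynamic feather bones.
# A small capsule on b_AR_Root anchors a large tail-feathers body, which in turn parents seven individual feather bodies and an upper feathers body.
# Constraints form a two-level hierarchy with tight swing and twist limits of a few degrees, keeping feather motion subtle, with the upper feathers the most restricted.
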PhysicsAsset /BRCosmetics/Weapons/FORT_Melee/Pickaxe_FabulousWind/Meshes/FabulousWind_Axe_Physics
bodies {
 b_AR_Root [{shapeType=Capsule, length=0.505, radius=0.505, rotation=<rot 0 -90 0>}]
 dyn_Feather_01 [{
   shapeType=Capsule
   length=18.0089
   radius=3.4348
   shapeName=dyn_Feather_01_capsule
   center=<-11.3556 0.3713 3.4511>
   rotation=<rot 0.0001 -88.8236 -71.3153>
  }]
 dyn_Feather_02 [{
   shapeType=Capsule
   length=10.0003
   radius=2.4438
   shapeName=dyn_Feather_02_capsule
   center=<2.1957 1.3145 1.2376>
   rotation=<rot 0 86.0155 -89.3956>
  }]
 dyn_Feather_03 [{
   shapeType=Capsule
   length=7.6486
   radius=2.0611
   shapeName=dyn_Feather_03_capsule
   center=<-7.2897 4.5556 0>
   rotation=<rot 0 -125.0172 -76.7894>
  }]
 dyn_Feather_04 [{
   shapeType=Capsule
   length=11.0722
   radius=2.5315
   shapeName=dyn_Feather_04_capsule
   center=<-16.7035 0.7275 4.1825>
   rotation=<rot 0 -85.178 -61.8129>
  }]
 dyn_Feather_05 [{shapeType=Capsule, length=11.927, radius=2.1256, shapeName=dyn_Feather_05_capsule, center=<-16.2407 0 0>, rotation=<rot 0 -83.0672 -89.4354>}]
 dyn_Feather_06 [{
   shapeType=Capsule
   length=6.1221
   radius=1.9381
   shapeName=dyn_Feather_06_capsule
   center=<-6.5338 -3.668 0>
   rotation=<rot -0.4607 -63.4962 -64.7221>
  }]
 dyn_Feather_07 [{
   shapeType=Capsule
   length=12.9604
   radius=2.0313
   shapeName=dyn_Feather_07_capsule
   center=<-13.0791 0 2.9093>
   rotation=<rot 0 -90.5323 -78.7914>
  }]
 dyn_Feathers_up [{
   shapeType=Capsule
   length=15.2204
   radius=2.4909
   shapeName=dyn_Feathers_up_capsule
   center=<2.0174 1.7795 -0.2609>
   rotation=<rot 0 -85.8332 -78.6849>
  }]
 dyn_tailFeathers [{shapeType=Capsule, length=29.7523, radius=17.2484, center=<-24.5593 0 5.2543>, rotation=<rot 0 -90.7424 -77.7423>}]
}
constraints [
 {bone1Name=dyn_tailFeathers, bone2Name=b_AR_Root, swing1LimitDegrees=5, swing2LimitDegrees=5, twistLimitDegrees=2.4}
 {bone1Name=dyn_Feather_01, bone2Name=dyn_tailFeathers, swing1LimitDegrees=6, swing2LimitDegrees=5, twistLimitDegrees=4}
 {bone1Name=dyn_Feather_02, bone2Name=dyn_tailFeathers, swing1LimitDegrees=6, swing2LimitDegrees=5, twistLimitDegrees=4}
 {bone1Name=dyn_Feather_03, bone2Name=dyn_tailFeathers, swing1LimitDegrees=6, swing2LimitDegrees=5, twistLimitDegrees=4}
 {bone1Name=dyn_Feather_04, bone2Name=dyn_tailFeathers, swing1LimitDegrees=6, swing2LimitDegrees=5, twistLimitDegrees=4}
 {bone1Name=dyn_Feather_05, bone2Name=dyn_tailFeathers, swing1LimitDegrees=6, swing2LimitDegrees=5, twistLimitDegrees=4}
 {bone1Name=dyn_Feather_06, bone2Name=dyn_tailFeathers, swing1LimitDegrees=6, swing2LimitDegrees=5, twistLimitDegrees=4}
 {bone1Name=dyn_Feather_07, bone2Name=dyn_tailFeathers, swing1LimitDegrees=6, swing2LimitDegrees=5, twistLimitDegrees=4}
 {bone1Name=dyn_Feathers_up, bone2Name=dyn_tailFeathers, swing1LimitDegrees=2, swing2LimitDegrees=2, twistLimitDegrees=2}
]
